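# Material function in the ImpostorBaker internals that blends three Vector4 inputs (R, G, B) into one value using a Float3 Weights input.
# It splits Weights into components, multiplies each input by its matching component and sums the products.
# The preview weights default to fully favoring the R input.
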MaterialFunction /ImpostorBaker/MaterialFunctions/Internal/FrameBlendWeights
graph {nodes [
  n1 = MaterialFunctionCall(Float3=FunctionInput() {bUsePreviewValueAsDefault=true, inputName=Weights, sortPriority=3, previewValue=<1 0 0 1>}) {materialFunction=&BreakOutFloat3Components}
  n2 = Multiply(A=FunctionInput() {inputName=R, inputType=FunctionInput_Vector4}, B=@n1.R)
  n3 = Multiply(A=FunctionInput() {inputName=G, inputType=FunctionInput_Vector4, sortPriority=1}, B=@n1.G)
  n4 = FunctionInput() {inputName=B, inputType=FunctionInput_Vector4, sortPriority=2}
  FunctionOutput(None=Add(A=Add(A=@n2, B=@n3), B=Multiply(A=@n4, B=@n1.B)))
 ]}
