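# Internal ImpostorBaker material function that samples one texture object at three frame UVs (R, G, B) and blends the results for impostor rendering.
# All three samples take explicit DDX/DDY from a separate Derivative UVs input, so mip selection is independent of the per-frame UVs.
# The appended samples and a Weights input feed FrameBlendWeights, which produces the main output.
# A second output exposes the unblended R sample.
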MaterialFunction /ImpostorBaker/MaterialFunctions/Internal/FrameBlendWeights_TextureObject
graph {nodes [
  n1 = FunctionInput() {inputName=Texture Object, inputType=FunctionInput_Texture2D}
  n2 = FunctionInput(Preview=TextureCoordinate()) {bUsePreviewValueAsDefault=true, inputName=Derivative UVs, inputType=FunctionInput_Vector2, sortPriority=10}
  n3 = DDX(Value=@n2)
  n4 = DDY(Value=@n2)
  n5 = TextureSample!(
   UVs=FunctionInput() {inputName=R UV, inputType=FunctionInput_Vector2, sortPriority=2}
   Tex=@n1
   "DDX(UVs)"=@n3
   "DDY(UVs)"=@n4
   Apply View MipBias
  ) {mipValueMode=TMVM_Derivative, samplerSource=SSM_Wrap_WorldGroupSettings}
  n6 = AppendVector(A=@n5.RGB, B=@n5.RGB)
  n7 = TextureSample(UVs=FunctionInput() {inputName=G UV, inputType=FunctionInput_Vector2, sortPriority=3}, Tex=@n1, "DDX(UVs)"=@n3, "DDY(UVs)"=@n4) {mipValueMode=TMVM_Derivative, samplerSource=SSM_Wrap_WorldGroupSettings}
  n8 = TextureSample(UVs=FunctionInput() {inputName=B UV, inputType=FunctionInput_Vector2, sortPriority=4}, Tex=@n1, "DDX(UVs)"=@n3, "DDY(UVs)"=@n4) {mipValueMode=TMVM_Derivative, samplerSource=SSM_Wrap_WorldGroupSettings}
  FunctionOutput(None=MaterialFunctionCall!(
    R=@n6
    G=AppendVector(A=@n7.RGB, B=@n7.RGB)
    B=AppendVector(A=@n8.RGB, B=@n8.RGB)
    Weights=FunctionInput() {bUsePreviewValueAsDefault=true, inputName=Weights, sortPriority=5, previewValue=<1 0 0 1>}
   ) {materialFunction=&/FrameBlendWeights}.Result)
  FunctionOutput(None=@n6) {outputName=R Sample, sortPriority=1}
 ]}
