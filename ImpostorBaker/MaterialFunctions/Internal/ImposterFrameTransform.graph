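# Material function for the Impostor Baker that builds an orthonormal frame around a Z vector and expresses a negated input vector in that frame.
# By default, X is a normalized horizontal perpendicular of Z and Y is the normalized cross product of Z and X, though both can be overridden.
# InverseTransformMatrix produces the unnormalized Local Z Vector, output alongside the three world basis vectors.
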MaterialFunction /ImpostorBaker/MaterialFunctions/Internal/ImposterFrameTransform
graph {
 properties {bExposeToLibrary=true}
 nodes [
  n1 = FunctionInput() {bUsePreviewValueAsDefault=true, inputName=Z Vec, sortPriority=5, previewValue=<0 0 1 1>}
  n2 = MaterialFunctionCall(Float3=@n1) {materialFunction=&BreakOutFloat3Components}
  n3 = Multiply(A=@n2.R) {constB=-1}
  n4 = Constant()
  n5 = FunctionInput(Preview=Normalize(VectorInput=MaterialFunctionCall(X=@n2.G, Y=@n3, Z=@n4) {materialFunction=&MakeFloat3}.Result) {desc=X}) {bUsePreviewValueAsDefault=true, inputName=X Vec, sortPriority=1, previewValue=<0 0 1 1>}
  n6 = FunctionInput(Preview=Normalize(VectorInput=CrossProduct(A=@n1, B=@n5)) {desc=Y}) {bUsePreviewValueAsDefault=true, inputName=Y Vec, sortPriority=2, previewValue=<0 0 1 1>}
  n7 = MaterialFunctionCall!(
   VectorToTransform=Multiply(A=FunctionInput() {bUsePreviewValueAsDefault=true, inputName=InVec, previewValue=<0 0 1 1>}) {constB=-1}
   BasisX=@n5
   BasisY=@n6
   BasisZ=@n1
  ) {materialFunction=/Engine/Functions/Engine_MaterialFunctions02/Math/InverseTransformMatrix}
  FunctionOutput(None=@n7.Transformed Vector 3) {outputName=Local Z Vector}
  FunctionOutput(None=@n5) {outputName=World X, sortPriority=1}
  FunctionOutput(None=@n6) {outputName=World Y, sortPriority=2}
  FunctionOutput(None=@n1) {outputName=World Z, sortPriority=3}
  Normalize(VectorInput=@n7.Transformed Vector 3)
 ]
}
