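# Material function in ImpostorBaker that decodes a 2D hemi-octahedral coordinate into a normalized 3D direction, exposed to the material library.
# A single Custom HLSL node rotates and scales the input into the octahedral diamond via half sum and difference, derives Z as one minus the L1 length, then normalizes.
# The math applies no 0–1 remap, so inputs in the -1 to 1 square map onto the +Z hemisphere.
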MaterialFunction /ImpostorBaker/MaterialFunctions/Internal/HemiOctahedronToUnitVector
graph {
 properties {bExposeToLibrary=true}
 nodes [
  MaterialExpressionFunctionInput_0 = FunctionInput() {inputName=Hemi Octahedron, inputType=FunctionInput_Vector2}
  n1 = Custom!(Oct=@MaterialExpressionFunctionInput_0) {
   code="\tOct = float2( Oct.x + Oct.y, Oct.x - Oct.y ) * 0.5;\r\n\tfloat3 N = float3( Oct, 1 - dot( 1, abs(Oct) ) );\r\n\treturn normalize(N);"
   inputs [{inputName=Oct, input {expression=&self:MaterialExpressionFunctionInput_0}}]
  }
  FunctionOutput(None=@n1) {outputName=Unit Vector}
 ]
}
