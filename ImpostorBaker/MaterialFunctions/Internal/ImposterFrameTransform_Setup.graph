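# Internal impostor material function that converts a 2D frame index into the World Z direction vector for that frame.
# Divides Frame by XY Frames minus one, then bias-scales it from the 0 to 1 range into -1 to 1.
# Static switch UpperHemisphereOnlyImpostor, on by default, picks hemi-octahedron or full octahedron decoding to a unit vector.
# A Normalize node on the hemisphere result is left unconnected to the output.
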MaterialFunction /ImpostorBaker/MaterialFunctions/Internal/ImposterFrameTransform_Setup
graph {
 properties {bExposeToLibrary=true}
 nodes [
  n1 = FunctionInput() {bUsePreviewValueAsDefault=true, inputName=XY Frames, inputType=FunctionInput_Vector2, sortPriority=2, previewValue=<16 16 0 1>}
  n2 = ConstantBiasScale(None=Divide(A=FunctionInput() {bUsePreviewValueAsDefault=true, inputName=Frame, inputType=FunctionInput_Vector2, sortPriority=1}, B=Subtract(A=@n1))) {bias=-0.5, scale=2}
  n3 = MaterialFunctionCall!(Hemi Octahedron=@n2) {materialFunction=&/HemiOctahedronToUnitVector}
  FunctionOutput(None=StaticSwitchParameter(
    True=@n3.Unit Vector
    False=MaterialFunctionCall!(Octahedron=@n2) {materialFunction=/Engine/Functions/Engine_MaterialFunctions01/Vectors/OctahedronToUnitVector}.Unit Vector
   ) {defaultValue=true, parameterName=UpperHemisphereOnlyImpostor}) {outputName=World Z, sortPriority=3}
  Normalize(VectorInput=@n3.Unit Vector) {desc=Z}
 ]
}
