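# Material function for the Impostor Baker that supplies an impostor's position and size to materials.
# It transforms the local origin to world space, adds a zero-valued direction term, and outputs that as the base position.
# Position also adds the transformed Pivot Offset parameter multiplied by the object's X scale.
# Default Mesh Size is output raw and multiplied by X scale alone, so sizing ignores non-uniform Y and Z scale.
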MaterialFunction /ImpostorBaker/MaterialFunctions/Internal/ImpostorInfo
graph {nodes [
  n1 = TransformPosition(None=Constant3Vector()) {transformType=TRANSFORMPOSSOURCE_World}
  n2 = Multiply(
   A=Transform(None=VectorParameter() {group=Impostor Setup, parameterName=Pivot Offset}.RGB)
   B=MaterialFunctionCall() {materialFunction=/Engine/Functions/Engine_MaterialFunctions02/WorldPositionOffset/ObjectScale}.Scale X
  )
  n3 = Transform(None=Constant3Vector()) {transformSourceType=TRANSFORMSOURCE_Local}
  FunctionOutput(None=Add(A=Add(A=@n1, B=@n2), B=@n3)) {outputName=Position}
  n4 = ScalarParameter() {defaultValue=350, group=Impostor Setup, parameterName=Default Mesh Size}
  FunctionOutput(None=Multiply(
    A=FunctionOutput(None=@n4) {outputName=Default Size, sortPriority=1}
    B=MaterialFunctionCall() {materialFunction=&ObjectScale}.Scale X
   )) {description=Multiplies Object Scale with the Stored UV Transform, outputName=Scaled Size, sortPriority=2}
  FunctionOutput(None=Add(A=@n1, B=@n3)) {outputName=OutputBasePosition}
 ]}
