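# OverrideMobile is an internal ImpostorBaker material function that selects an alternate value on mobile feature levels.
# A Feature Level Switch outputs the Mobile Override input for ES2 and ES3_1, and the Default input for all other levels.
# Both inputs use their preview value as the default, so either can be left unconnected.
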MaterialFunction /ImpostorBaker/MaterialFunctions/Internal/OverrideMobile
graph {nodes [
  n1 = FunctionInput() {bUsePreviewValueAsDefault=true, inputName=Default}
  n2 = FunctionInput() {bUsePreviewValueAsDefault=true, inputName=Mobile Override, sortPriority=1}
  FunctionOutput(None=FeatureLevelSwitch(Default=@n1, ES2=@n2, ES3_1=@n2, SM4_REMOVED=@n1))
 ]}
