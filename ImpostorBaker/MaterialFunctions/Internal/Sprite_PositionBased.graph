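# Impostor Baker material function that billboards a quad mesh toward the camera position, or a supplied ray origin, around a center location.
# It builds X and Y axes by crossing the center-to-origin direction with world Z, then scales centered UVs along them to place vertices.
# Outputs are absolute position, a World Position Offset, and optional normals transformed from view to tangent space.
# An unconnected FramesXY parameter branch remains.
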MaterialFunction /ImpostorBaker/MaterialFunctions/Internal/Sprite_PositionBased
graph {
 properties {
  bExposeToLibrary=true
  description=Use this function within the world position offset input to align plane meshes with the screen at any location.
  libraryCategoriesText [Misc, World Position Offset]
 }
 nodes [
  n1 = FunctionInput(Preview=CameraPositionWS()) {
   bUsePreviewValueAsDefault=true
   description=The world position of the center of the sprite
   inputName=Ray Origin
   sortPriority=5
   previewValue=<0 0 0 0>
  }
  n2 = FunctionInput(Preview=MaterialFunctionCall() {materialFunction=&ComponentPivotLocation}.Result) {
   bUsePreviewValueAsDefault=true
   description=The world position of the center of the sprite
   inputName=Center Location
   previewValue=<0 0 0 0>
  }
  n3 = Normalize(VectorInput=Subtract(A=@n1, B=@n2))
  n4 = Normalize(VectorInput=CrossProduct(A=@n3, B=Constant3Vector() {constant=<rgba 0 0 1 0>})) {desc=X}
  n5 = FunctionInput(Preview=TextureCoordinate()) {
   bUsePreviewValueAsDefault=true
   description=The quad should supply uvs that are mapped from the 0 to 1 uv range.
   inputName=0-1 UVs
   inputType=FunctionInput_Vector2
   previewValue=<0 0 0 0>
  }
  n6 = MaterialFunctionCall(Float2=ConstantBiasScale(None=@n5) {bias=-0.5, scale=1}) {materialFunction=&BreakOutFloat2Components}
  n7 = FunctionInput() {bUsePreviewValueAsDefault=true, inputName=XY scale, inputType=FunctionInput_Vector2, previewValue=<50 100 0 1>}
  n8 = MaterialFunctionCall(Float2=@n7) {materialFunction=&BreakOutFloat2Components}
  n9 = Multiply(A=Multiply(A=@n4, B=@n6.R), B=@n8.R)
  n10 = CrossProduct(A=@n3, B=@n4)
  n11 = Multiply(A=Multiply(A=Normalize(VectorInput=@n10) {desc=Y}, B=@n6.G), B=@n8.G)
  n12 = FunctionOutput(None=Add(A=Add(A=@n9, B=@n11), B=@n2)) {
   description=Position. This is the World Position of the Vertices. Should not be used for WorldPositionOffset becaues it does not have the original positions subtracted.
   outputName=World Position
   sortPriority=1
  }
  FunctionOutput(None=Subtract(A=@n12, B=WorldPosition() {worldPositionShaderOffset=WPT_ExcludeAllShaderOffsets}.XYZ)) {
   description=Plug this into the world position offset input. Please note that world position is already subtracted from the output.
   outputName=World position offset
  }
  n13 = FunctionInput() {
   bUsePreviewValueAsDefault=true
   description=This is an optional input. The function will correctly transform your normals.
   inputName="Normals (Optional)"
   sortPriority=10
   previewValue=<0 0 1 1>
  }
  n14 = Constant3Vector() {constant=<rgba 1 -1 -1 0>}
  FunctionOutput(None=Transform(None=Multiply(A=@n13, B=@n14)) {transformSourceType=TRANSFORMSOURCE_View, transformType=TRANSFORM_Tangent}) {description=This will output transformed tangent space normals., outputName=Normals, sortPriority=2}
  Multiply(A=TextureCoordinate() {uTiling=10, vTiling=10}, B=ScalarParameter() {defaultValue=16, parameterName=FramesXY})
 ]
}
