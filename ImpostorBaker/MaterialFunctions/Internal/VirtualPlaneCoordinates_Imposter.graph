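# Impostor Baker material function that raytraces a virtual plane and outputs its UVs and hit position.
# It intersects the ray (by default camera to pixel world position) with a plane through the coordinate origin, given by a normal.
# The hit is projected onto plane X/Y axes, derived from the normal unless supplied, then divided by UV Scale and centered at 0.5.
# Non-positive hit distances collapse the UVs to the center.
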MaterialFunction /ImpostorBaker/MaterialFunctions/Internal/VirtualPlaneCoordinates_Imposter
graph {
 properties {bExposeToLibrary=true, description="Virtual coordinates for a raytraced plane, in world units", libraryCategoriesText [Misc,Coordinates]}
 nodes [
  n1 = NamedRerouteUsage()
  n2 = FunctionInput() {bUsePreviewValueAsDefault=true, inputName=Plane Normal Axis, sortPriority=2, previewValue=<0 0 1 1>}
  n3 = FunctionInput(Preview=CameraPositionWS() {desc=Ray Origin}) {
   bUsePreviewValueAsDefault=true
   desc=Plane Origin
   description=The center location of the virtual Plane
   inputName=Ray Origin Local
   sortPriority=5
  }
  n4 = FunctionInput(Preview=Subtract(A=WorldPosition().XYZ, B=NamedRerouteDeclaration(None=@n3) {name=Ray Origin Local, nodeColor=<rgba 0.875 1 0 1>}) {desc=Direction Vector}) {
   bUsePreviewValueAsDefault=true
   desc=Plane Dir
   description=The center location of the virtual Plane
   inputName=Ray Dir Local
   sortPriority=4
  }
  n5 = Divide(A=Multiply(A=DotProduct(A=@n2, B=@n1)) {constB=-1}, B=DotProduct(A=@n2, B=@n4) {desc=VectorDotPlaneNormal}) {desc=Intersection Time1}
  n6 = MaterialFunctionCall!(Vector1=@n2, Vector2=Constant3Vector() {constant=<rgba 0 0.0002 -1 0>}) {materialFunction=/Engine/Functions/Engine_MaterialFunctions02/Math/CreateThirdOrthogonalVector}
  n7 = FunctionInput(Preview=Multiply(A=@n6.Vector3) {desc=Plane X}) {
   bUsePreviewValueAsDefault=true
   description="By default, the x axis or tangent vector is defined as the cross product of the Normal and 0,0.00015,1. You can specify a manual tangent vector here."
   inputName=Plane X Axis
   previewValue=<0 0 1 1>
  }
  n8 = Add(A=Multiply(A=@n5, B=@n4), B=NamedRerouteUsage()) {desc=Intersection Position}
  n9 = DotProduct(A=@n7, B=@n8)
  n10 = FunctionInput(Preview=Normalize(VectorInput=CrossProduct(A=@n7, B=@n6.Vector1)) {desc=Plane Y}) {
   bUsePreviewValueAsDefault=true
   description="By default, the x axis or tangent vector is defined as the cross product of the Normal and 0,0.00015,1. You can specify a manual tangent vector here."
   inputName=Plane Y Axis
   sortPriority=1
   previewValue=<0 0 1 1>
  }
  n11 = DotProduct(A=@n10, B=@n8)
  n12 = Constant()
  n13 = ComponentMask(None=If(A=@n5, B=Constant(), "A > B"=AppendVector(A=@n9, B=@n11), "A == B"=@n12, "A < B"=@n12)) {g=true, r=true}
  n14 = FunctionInput() {
   bUsePreviewValueAsDefault=true
   description=The size of the virtual plane
   inputName=UV Scale
   inputType=FunctionInput_Vector2
   sortPriority=3
   previewValue=<1 1 0 1>
  }
  FunctionOutput(None=Add(A=Divide(A=@n13, B=@n14)) {constB=0.5}) {description=The 2D UVs of the virtual plane coordinates., outputName=UVs}
  FunctionOutput(None=@n8) {description=The world position at each point on the virtual plane., outputName=World Position, sortPriority=1}
  Multiply(A=@n10) {constB=-1}
 ]
}
