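# Debug material function for octahedral impostors, outputting base color, normal, opacity mask, WPO, pixel depth offset and frame-axis debug values.
# Camera direction selects three neighboring atlas frames, projected onto virtual planes, parallax-shifted by texture alpha and weight-blended.
# Sprite WPO billboards the mesh; mobile overrides use a single-frame version.
# Shadow passes derive rays from sun direction and skip parallax; static switches add frame-grid overlays and manual weights.
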MaterialFunction /ImpostorBaker/MaterialFunctions/Internal/ImpostorDebugLocal
graph {nodes [
  n1 = MaterialFunctionCall!(Frame=NamedRerouteUsage(), XY Frames=NamedRerouteUsage()) {materialFunction=&/ImposterFrameTransform_Setup}
  FunctionOutput(None=@n1.World Z) {outputName=DEBUG, sortPriority=9}
  n2 = MaterialFunctionCall!(InVec=NamedRerouteUsage(), X Vec, Y Vec, Z Vec=@n1.World Z) {materialFunction=&/ImposterFrameTransform}
  n3 = MaterialFunctionCall!(
   Plane X Axis=@n2.World X
   Plane Y Axis=@n2.World X
   Plane Normal Axis=@n2.World X
   UV Scale=NamedRerouteUsage()
   Ray Dir Local=NamedRerouteUsage()
   Ray Origin Local=NamedRerouteUsage()
  ) {materialFunction=&/VirtualPlaneCoordinates_Imposter}
  n4 = Add(A=NamedRerouteUsage() {desc="0,0"}, B=Divide(A=@n3.UVs, B=NamedRerouteUsage()))
  MaterialExpressionSubtract_12 = Subtract(B=TextureSample(UVs=@n4, Tex=NamedRerouteUsage()) {texture=&Black}.A) {constA=0.5, constB=0.5}
  MaterialExpressionConstant_5 = Constant()
  n5 = Custom!(Main=@MaterialExpressionSubtract_12, Shadow=@MaterialExpressionConstant_5) {
   code="##if SHADOW_DEPTH_SHADER \r\n{\r\nreturn Shadow;\r\n}\r\n##endif\r\nreturn Main;"
   desc=Optimize Shadows
   description=Shadow Pass Switch
   outputType=CMOT_Float2
   inputs [
    {inputName=Main, input {expression=&self:MaterialExpressionSubtract_12}}
    {inputName=Shadow, input {expression=&self:MaterialExpressionConstant_5}}
   ]
  }
  n6 = VertexInterpolator(VS=Divide(A=ComponentMask(None=@n2.Local Z Vector) {g=true, r=true}, B=NamedRerouteUsage()) {constB=16})
  n7 = Add(A=@n4, B=Multiply(A=Multiply(A=@n5, B=@n6.PS), B=NamedRerouteUsage()))
  n8 = Constant2Vector() {g=1, r=1}
  n9 = NamedRerouteUsage()
  n10 = NamedRerouteUsage()
  n11 = Constant2Vector() {g=1, r=1}
  n12 = MaterialFunctionCall(
   InVec=NamedRerouteUsage()
   Z Vec=MaterialFunctionCall(Frame=Add(A=@n10, B=@n11), XY Frames=NamedRerouteUsage()) {materialFunction=&ImposterFrameTransform_Setup}.World Z
  ) {materialFunction=&ImposterFrameTransform}
  n13 = MaterialFunctionCall(
   Plane X Axis=@n12.World X
   Plane Y Axis=@n12.World X
   Plane Normal Axis=@n12.World X
   UV Scale=NamedRerouteUsage()
   Ray Dir Local=NamedRerouteUsage()
   Ray Origin Local=NamedRerouteUsage()
  ) {materialFunction=&VirtualPlaneCoordinates_Imposter}
  n14 = Add(A=Add(A=NamedRerouteUsage() {desc="1,1"}, B=Divide(A=@n8, B=@n9)), B=Divide(A=@n13.UVs, B=NamedRerouteUsage()))
  n15 = Subtract(B=TextureSample(UVs=@n14, Tex=NamedRerouteUsage()) {texture=&Black}.A) {constA=0.5, constB=0.5}
  n16 = VertexInterpolator(VS=Divide(A=ComponentMask(None=@n12.Local Z Vector) {g=true, r=true}, B=NamedRerouteUsage()) {constB=16})
  n17 = Add(A=@n14, B=Multiply(A=Multiply(A=@n15, B=@n16.PS), B=NamedRerouteUsage()))
  n18 = LinearInterpolate(A=Constant2Vector() {g=1}, B=Constant2Vector() {r=1}, Alpha=NamedRerouteUsage())
  n19 = NamedRerouteUsage()
  n20 = NamedRerouteUsage()
  n21 = LinearInterpolate(A=Constant2Vector() {g=1}, B=Constant2Vector() {r=1}, Alpha=NamedRerouteUsage())
  n22 = MaterialFunctionCall(
   InVec=NamedRerouteUsage()
   Z Vec=MaterialFunctionCall(Frame=Add(A=@n20, B=@n21), XY Frames=NamedRerouteUsage()) {materialFunction=&ImposterFrameTransform_Setup}.World Z
  ) {materialFunction=&ImposterFrameTransform}
  n23 = MaterialFunctionCall(
   Plane X Axis=@n22.World X
   Plane Y Axis=@n22.World X
   Plane Normal Axis=@n22.World X
   UV Scale=NamedRerouteUsage()
   Ray Dir Local=NamedRerouteUsage()
   Ray Origin Local=NamedRerouteUsage()
  ) {materialFunction=&VirtualPlaneCoordinates_Imposter}
  n24 = Add(A=Add(A=NamedRerouteUsage() {desc="1,0 or 0,1"}, B=Divide(A=@n18, B=@n19)), B=Divide(A=@n23.UVs, B=NamedRerouteUsage()))
  n25 = Subtract(B=TextureSample(UVs=@n24, Tex=NamedRerouteUsage()) {texture=&Black}.A) {constA=0.5, constB=0.5}
  n26 = VertexInterpolator(VS=Divide(A=ComponentMask(None=@n22.Local Z Vector) {g=true, r=true}, B=NamedRerouteUsage()) {constB=16})
  n27 = Add(A=@n24, B=Multiply(A=Multiply(A=@n25, B=@n26.PS), B=NamedRerouteUsage()))
  n28 = MaterialFunctionCall!(Texture Object=NamedRerouteUsage(), R UV=@n7, G UV=@n27, B UV=@n17, Weights=NamedRerouteUsage(), Derivative UVs=NamedRerouteUsage()) {desc=Color and Alpha, materialFunction=&/FrameBlendWeights_TextureObject}
  n29 = ComponentMask(None=@n28.Result) {b=true, g=true, r=true}
  n30 = NamedRerouteUsage()
  FunctionOutput(None=LinearInterpolate(
    A=MaterialFunctionCall!(Default=@n29, Mobile Override=@n30) {materialFunction=&/OverrideMobile}.Result
    B=NamedRerouteUsage()
    Alpha=NamedRerouteUsage()
   )) {outputName=BaseColor}
  n31 = MaterialFunctionCall(Texture Object=NamedRerouteUsage(), R UV=@n7, G UV=@n27, B UV=@n17, Weights=NamedRerouteUsage(), Derivative UVs=NamedRerouteUsage()) {desc=Normal and Depth, materialFunction=&FrameBlendWeights_TextureObject}
  n32 = ConstantBiasScale(None=MaterialFunctionCall(Default=ComponentMask(None=@n31.Result) {b=true, g=true, r=true}, Mobile Override=NamedRerouteUsage()) {materialFunction=&OverrideMobile}.Result) {bias=-0.5, scale=2}
  FunctionOutput(None=LinearInterpolate(A=Transform(None=@n32), B=Constant3Vector() {constant=<rgba 0 0 1 0>}, Alpha=NamedRerouteUsage())) {outputName=Normal, sortPriority=2}
  n33 = CameraPositionWS()
  n34 = Divide(
   A=Subtract(A=ComponentMask(None=@n33) {b=true}, B=ScalarParameter() {defaultValue=10000, parameterName=Camera Z Fadeout Loc})
   B=ScalarParameter() {defaultValue=35000, parameterName=Cam Z Fade Length}
  )
  n35 = Add(
   A=Multiply(A=VectorParameter() {parameterName=MobileZOffset, defaultValue=<rgba 0 0 550 0>}.RGB, B=Saturate(None=@n34))
   B=NamedRerouteUsage()
  )
  n36 = NamedRerouteUsage()
  FunctionOutput(None=MaterialFunctionCall(Default=NamedRerouteUsage(), Mobile Override=DistanceFieldsRenderingSwitch!(No=@n35, Yes=@n36)) {materialFunction=&OverrideMobile}.Result) {outputName=WPO, sortPriority=3}
  n37 = MaterialFunctionCall(Default=ComponentMask(None=@n31.Result) {a=true}, Mobile Override=NamedRerouteUsage()) {materialFunction=&OverrideMobile}
  n38 = CameraVectorWS()
  n39 = Abs(None=DotProduct(
    A=Normalize(VectorInput=@n38)
    B=MaterialFunctionCall() {materialFunction=&CameraDirectionVector}.Result
   ))
  n40 = MaterialFunctionCall() {materialFunction=&/ImpostorInfo}
  n41 = Multiply(
   A=OneMinus(None=ComponentMask(None=@n37.Result) {r=true})
   B=Multiply(A=Multiply(A=@n39, B=@n40.Scaled Size) {constB=0.5}, B=ScalarParameter() {defaultValue=1, parameterName=PDO})
  )
  n42 = OneMinus(None=IsOrthographic())
  n43 = Constant()
  FunctionOutput(None=QualitySwitch(Default=Multiply(A=@n41, B=@n42), Low=@n43, Medium=@n43)) {outputName=PDO, sortPriority=4}
  n44 = ShadowReplace(Default=@n28.Result, Shadow=@n28.Result)
  n45 = ComponentMask(None=MaterialFunctionCall(Default=ComponentMask(None=@n44) {a=true}, Mobile Override=NamedRerouteUsage()) {materialFunction=&OverrideMobile}.Result) {r=true}
  n46 = NamedRerouteUsage()
  FunctionOutput(None=Multiply(A=Add(A=@n45, B=@n46) {desc=Add Debug}, B=NamedRerouteUsage()) {desc="Clip debug box\r\nNeeded due to WPO overscan when debugging\r\nBecause the mesh cuts out the unused space"}) {outputName=Opacity Mask, sortPriority=1}
  FunctionOutput(None=VertexInterpolator(VS=@n1.World Z).PS) {outputName=DEBUG_VS, sortPriority=10}
  NamedRerouteDeclaration(None=FunctionInput(Preview=TextureObject() {samplerType=SAMPLERTYPE_LinearColor, texture=/Engine/Functions/Engine_MaterialFunctions02/PivotPainter2/Black_1x1_EXR_Texture}) {inputName=Normal, inputType=FunctionInput_Texture2D, sortPriority=1}) {name=NormalTex, nodeColor=<rgba 0.0937 0.4331 1 1>}
  n47 = MaterialFunctionCall!(Frames XY=NamedRerouteUsage(), BaseColor=NamedRerouteUsage(), Normal=NamedRerouteUsage()) {materialFunction=/ImpostorBaker/MaterialFunctions/Impostor_SingleFrameVersion}
  NamedRerouteDeclaration(None=@n47.BaseColor RGB) {name=Single Frame Color, nodeColor=<rgba 0 0 0 1>}
  NamedRerouteDeclaration(None=@n47.BaseColor A) {name=Single Frame Alpha, nodeColor=<rgba 0 0 0 1>}
  NamedRerouteDeclaration(None=@n47.Normal RGB) {name=Single Frame Normal, nodeColor=<rgba 0 0 0 1>}
  n48 = MaterialFunctionCall() {materialFunction=&ImpostorInfo}
  NamedRerouteDeclaration(None=@n48.Scaled Size) {name=Scaled Size, nodeColor=<rgba 0 1 0.9063 1>}
  n49 = NamedRerouteUsage()
  NamedRerouteDeclaration(None=ShadowReplace(Default=Floor(None=@n49), Shadow=Round(None=@n49))) {name=Frame Index, nodeColor=<rgba 0 0.1609 1 1>}
  n50 = CameraPositionWS()
  n51 = MaterialFunctionCall() {materialFunction=&ImpostorInfo}
  n52 = NamedRerouteUsage()
  n53 = Normalize(VectorInput=@n52) {desc=Vector to Geo}
  n54 = MaterialFunctionCall() {materialFunction=&CameraDirectionVector}
  n55 = Multiply(
   A=Subtract(
    A=Distance(A=@n50, B=@n51.Position)
    B=MaterialFunctionCall(Vector 3=@n52) {desc=Dist to Geo, materialFunction=/Engine/Functions/Engine_MaterialFunctions02/Utility/VectorLength}.V3 Length
   )
   B=Abs(None=DotProduct(A=@n53, B=@n54.Result))
  )
  NamedRerouteDeclaration(None=StaticSwitchParameter(True=Multiply(A=@n55, B=@n53), False=Constant()) {desc="Makes Pixel Depth Offset Slightly more Accurate\r\nNot typically necessary and adds cost", parameterName=Align to Camera Plane}) {name=Camera Plane Alignment, nodeColor=<rgba 1 0.375 0 1>}
  n56 = ScalarParameter() {defaultValue=1, parameterName=WPOScale}
  NamedRerouteDeclaration(None=StaticSwitchParameter(True=Multiply(A=@n56), False=@n56) {parameterName=Show Frame Debug Info}) {name=WPO Scale, nodeColor=<rgba 0.0634 0.7524 0.8646 1>}
  n57 = MaterialFunctionCall(Float3=NamedRerouteUsage()) {materialFunction=&BreakOutFloat3Components}
  n58 = Normalize(VectorInput=MaterialFunctionCall(X=@n57.R, Y=@n57.R, Z=Max(A=@n57.B) {constB=0.001}) {materialFunction=&MakeFloat3}.Result)
  n59 = StaticSwitchParameter(
   True=MaterialFunctionCall!(Unit Vector=@n58) {materialFunction=&/UnitVectorToHemiOctahedra}.Hemi Octahedron
   False=MaterialFunctionCall!(Vector=@n58) {materialFunction=/Engine/Functions/Engine_MaterialFunctions01/Vectors/UnitVectorToOctahedron}.Octahedron
  ) {
   defaultValue=true
   desc="Defaults to True. When True, the impostor is only captured to support views from the upper hemisphere. Otherwise it will include views from all views. Restricting to the upper hemisphere is useful for trees that are unlikely to be seen from below when far away."
   parameterName=UpperHemisphereOnlyImpostor
  }
  n60 = Saturate(None=ConstantBiasScale(None=@n59))
  n61 = NamedRerouteUsage()
  NamedRerouteDeclaration(None=Multiply(A=NamedRerouteDeclaration(None=@n60) {name=Octahedron, nodeColor=<rgba 0.1019 0.4322 1 1>}, B=Subtract(A=@n61))) {name=Octahedron Scaled, nodeColor=<rgba 0.1014 0.326 1 1>}
  n62 = Multiply(A=SkyAtmosphereLightDirection()) {constB=50000}
  n63 = NamedRerouteDeclaration(None=ShadowReplace(Default=CameraPositionWS(), Shadow=Add(A=@n62, B=@n48.Position))) {name=Ray Origin, nodeColor=<rgba 0.3359 1 0 1>}
  n64 = Subtract(A=@n63, B=MaterialFunctionCall() {materialFunction=&ImpostorInfo}.Position)
  n65 = Normalize(VectorInput=NamedRerouteDeclaration(None=Transform(None=@n64) {transformSourceType=TRANSFORMSOURCE_World, transformType=TRANSFORM_Tangent}) {name=Ray Origin Local, nodeColor=<rgba 0.2891 1 0 1>})
  NamedRerouteDeclaration(None=@n65) {name=Ray Origin Normalized, nodeColor=<rgba 0.4766 1 0 1>}
  n66 = Transform(None=Subtract(A=@n63, B=NamedRerouteUsage())) {transformSourceType=TRANSFORMSOURCE_World, transformType=TRANSFORM_Tangent}
  n67 = StaticSwitchParameter(True=Normalize(VectorInput=@n66), False=@n65) {
   desc="Defaults to True.  When True, the View Ray will be interpolated.\r\nWhen false, the view ray is treated as orthographic."
   parameterName=Interpolate Ray Position
  }
  NamedRerouteDeclaration(None=@n67) {name=Ray Origin Interpolated, nodeColor=<rgba 0.0853 0.3677 0.6302 1>}
  n68 = VertexInterpolator(VS=NamedRerouteUsage())
  n69 = MaterialFunctionCall!(UV=Frac(None=@n68.PS)) {materialFunction=&/TriangleInterpolator}
  NamedRerouteDeclaration(None=@n69.Mask) {name=Triangle Mask, nodeColor=<rgba 1 0.3802 0 1>}
  NamedRerouteDeclaration(None=MaterialFunctionCall(UV=Frac(None=@n49)) {materialFunction=&TriangleInterpolator}.Mask) {name=Triangle Mask VS, nodeColor=<rgba 1 0.4453 0 1>}
  n70 = NamedRerouteUsage()
  n71 = VectorParameter() {parameterName=Manual Weights, defaultValue=<rgba 1 0 0 1>}
  n72 = ScalarParameter() {defaultValue=1, parameterName=Manual Weights}
  NamedRerouteDeclaration(None=StaticSwitchParameter(True=LinearInterpolate(A=@n70, B=@n71.RGB, Alpha=@n72), False=NamedRerouteUsage()) {parameterName=Enable Manual Weights}) {name=Blend Weights, nodeColor=<rgba 0.0937 0.4331 1 1>}
  NamedRerouteDeclaration(None=Multiply(A=TextureCoordinate() {uTiling=10, vTiling=10}, B=ScalarParameter() {defaultValue=0.6, parameterName=Mip UV Scale})) {name=Derivative UVs, nodeColor=<rgba 0.0937 0.4331 1 1>}
  NamedRerouteDeclaration(None=StaticSwitchParameter(True=ScalarParameter() {parameterName=EnableParallax}, False=Constant() {r=1}) {defaultValue=true, desc=Useful to Toggle Parallax without recompiling, parameterName=Parallax Debug Toggle}) {name=Enable Parallax, nodeColor=<rgba 0 0.4375 1 1>}
  n73 = Multiply(A=NamedRerouteUsage(), B=Subtract(A=NamedRerouteUsage()))
  n74 = Multiply(A=NamedRerouteUsage(), B=Subtract(A=NamedRerouteUsage()))
  n75 = Multiply(
   A=SphereMask(A=Floor(None=@n73), B=Floor(None=@n74), Radius=ScalarParameter() {defaultValue=0.01, parameterName=FrameDebugRadius}) {desc=Isolate Current Frame}
   B=NamedRerouteUsage()
  )
  n76 = NamedRerouteUsage()
  n77 = NamedRerouteUsage()
  n78 = ScalarParameter() {defaultValue=0.015, parameterName=FrameDebugRadius2}
  n79 = MaterialFunctionCall(UVs=Multiply(A=TextureCoordinate() {uTiling=10, vTiling=10}, B=NamedRerouteUsage()), Texture Scale=Divide(B=NamedRerouteUsage())) {materialFunction=&ScaleUVsByCenter}
  n80 = NamedRerouteUsage()
  n81 = Frac(None=Multiply(A=NamedRerouteDeclaration(None=@n79.UVs) {name=NormalizedDebugUV, nodeColor=<rgba 1 0 0.5859 1>}, B=Subtract(A=@n80)))
  n82 = ScalarParameter() {defaultValue=0.03, parameterName=grid line}
  n83 = Add(
   A=Add(A=Multiply(A=@n75) {constB=0.5}, B=SphereMask(A=@n76, B=@n77, Radius=@n78) {desc=Draw Dot for Ray})
   B=Multiply(
    A=ScalarParameter() {defaultValue=1, parameterName=Debug Grid}
    B=MaterialFunctionCall!(UV=@n81, Line Width=@n82) {materialFunction=&/TriangulatedWireframe}.Result
   )
  )
  n84 = ScalarParameter() {defaultValue=1, parameterName=ShowDebug}
  NamedRerouteDeclaration(None=StaticSwitchParameter(True=Multiply(A=@n83, B=@n84), False=Constant()) {parameterName=Show Frame Debug Info}) {name=Frame Debug Display, nodeColor=<rgba 0 0 0 1>}
  n85 = MaterialFunctionCall(Float2=NamedRerouteUsage()) {materialFunction=&BreakOutFloat2Components}
  n86 = Constant() {r=0.5}
  n87 = Multiply(A=SphereMask(A=@n85.R, B=@n86) {attenuationRadius=0.5}, B=SphereMask(A=@n85.G, B=@n86) {attenuationRadius=0.5})
  NamedRerouteDeclaration(None=StaticSwitchParameter(True=@n87, False=Constant() {r=1}) {parameterName=Show Frame Debug Info}) {name=Debug Box Clip, nodeColor=<rgba 0 0 0 1>}
  n88 = Multiply(A=DotProduct(A=@n83, B=Constant() {r=2}), B=ScalarParameter() {defaultValue=1, parameterName=ShowDebug})
  NamedRerouteDeclaration(None=StaticSwitchParameter(True=Multiply(A=@n88, B=@n87), False=Constant()) {parameterName=Show Frame Debug Info}) {name=Frame DebugMask, nodeColor=<rgba 0 0 0 1>}
  n89 = MaterialFunctionCall(UV=Frac(None=@n74)) {materialFunction=&TriangleInterpolator}
  n90 = NamedRerouteUsage()
  n91 = Subtract(A=NamedRerouteUsage())
  n92 = MaterialFunctionCall(UV=Frac(None=Multiply(A=@n90, B=@n91))) {materialFunction=&TriangleInterpolator}
  n93 = LinearInterpolate(A=OneMinus(None=@n89.Mask), B=@n89.Mask, Alpha=@n92.Mask)
  n94 = Constant() {r=1}
  n95 = MaterialFunctionCall!(UVs=NamedRerouteUsage(), Tiling=ScalarParameter() {defaultValue=384, parameterName=DebugShadowCheckerDither}) {materialFunction=/Engine/ArtTools/RenderToTexture/MaterialFunctions/CheckerPattern}
  NamedRerouteDeclaration(None=Multiply(A=Multiply(A=@n93, B=@n92.Result), B=ShadowReplace(Default=@n94, Shadow=@n95.Result))) {name=Debug Triangle Mask, nodeColor=<rgba 1 0 0.375 1>}
  NamedRerouteDeclaration(None=FunctionInput(Preview=TextureObject() {samplerType=SAMPLERTYPE_LinearColor, texture=&Black_1x1_EXR_Texture}) {inputName=BaseColor, inputType=FunctionInput_Texture2D}) {name=BaseColorTex, nodeColor=<rgba 0.0937 0.4331 1 1>}
  n96 = Floor(None=@n68.PS)
  n97 = Round(None=@n68.PS)
  NamedRerouteDeclaration(None=Divide(A=ShadowReplace(Default=@n96, Shadow=@n97), B=NamedRerouteUsage())) {name=Octahedron Scaled VS, nodeColor=<rgba 0 0 0 1>}
  NamedRerouteDeclaration(None=@n69.Result) {name=Triangle Weights, nodeColor=<rgba 0.0593 0.4665 0.7656 1>}
  NamedRerouteDeclaration(None=@n47.Normal A) {name=SingleFrame PDO, nodeColor=<rgba 0 0 0 1>}
  NamedRerouteDeclaration(None=@n47.WPO) {name=SingleFrame WPO, nodeColor=<rgba 0 0 0 1>}
  n98 = FunctionInput() {bUsePreviewValueAsDefault=true, inputName=Frames XY, inputType=FunctionInput_Scalar, previewValue=<16 0 0 1>}
  n99 = Multiply(A=TextureCoordinate() {uTiling=10, vTiling=10}, B=NamedRerouteDeclaration(None=@n98) {name=Frames XY, nodeColor=<rgba 1 0 0.4922 1>})
  n100 = MaterialFunctionCall() {materialFunction=&ImpostorInfo}
  n101 = Multiply(A=@n100.Scaled Size, B=NamedRerouteUsage())
  n102 = StaticSwitchParameter(
   True=MaterialFunctionCall!(Center Location=@n100.Position, 0-1 UVs=@n99, XY scale=@n101, Ray Origin=NamedRerouteUsage(), "Normals (Optional)") {materialFunction=&/Sprite_PositionBased}.World position offset
   False=MaterialFunctionCall!(Center Location=@n100.Position, 0-1 UVs=@n99, XY scale=@n101, OverridePivot, "Normals (Optional)") {materialFunction=&Sprite}.World position offset
  ) {
   defaultValue=true
   desc="When set to True, the sprite orientation will be derived from the position, rather than assumed to be the view camera transform. Provides more accuracy for wide FOV but costs slightly more. Defaults ot True."
   parameterName=Derive Sprite Transform
  }
  n103 = CameraPositionWS()
  n104 = NamedRerouteDeclaration(None=Add(A=@n102, B=WorldPosition() {worldPositionShaderOffset=WPT_ExcludeAllShaderOffsets}.XYZ)) {name=WorldPos+Sprite, nodeColor=<rgba 0 0 0 1>}
  n105 = Normalize(VectorInput=NamedRerouteDeclaration(None=Subtract(A=@n103, B=@n104)) {name=Camera-to-Vertex, nodeColor=<rgba 0 0 0 1>}) {desc=CameraVector}
  n106 = Multiply(A=MaterialFunctionCall() {materialFunction=&ImpostorInfo}.Scaled Size) {constB=0.5}
  n107 = Subtract(A=Multiply(A=Multiply(A=@n105, B=@n106), B=ScalarParameter() {defaultValue=1, parameterName=Camera Push}), B=NamedRerouteUsage())
  n108 = DistanceFieldsRenderingSwitch(No=Constant(), Yes=@n107)
  n109 = FeatureLevelSwitch(Default=@n107, ES2=@n108, ES3_1=@n108, SM4_REMOVED=@n107, SM5=@n107, SM6=@n107)
  n110 = Constant()
  NamedRerouteDeclaration(None=Add(A=@n102, B=ShadowReplace(Default=@n109, Shadow=@n110)) {constB=0}) {name=WorldPositionOffset, nodeColor=<rgba 1 0.0234 0 1>}
  NamedRerouteDeclaration(None=Multiply(A=@n66) {constB=-1}) {name=Ray Dir Local, nodeColor=<rgba 0 0.1094 1 1>}
  Saturate(None=VertexInterpolator(VS=@n13.UVs).PS)
  Saturate(None=@n3.UVs)
  Saturate(None=VertexInterpolator(VS=@n23.UVs).PS)
  ShaderStageSwitch()
  VertexInterpolator(VS=@n3.UVs)
 ]}
