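# Impostor Baker material function that transforms a vector, world position by default, into a frame built around a supplied Z vector.
# It derives X from the normalized Z crossed with world up, then Y from Z crossed with X.
# The input minus an optional center location passes through InverseTransformMatrix with that basis.
# No fallback handles a vertical Z, where the cross product vanishes; the default Z is slightly tilted.
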MaterialFunction /ImpostorBaker/MaterialFunctions/Internal/TransformToFixedFrame
graph {
 properties {
  bExposeToLibrary=true
  description=This transforms a vector based on a Z vector and generated orthogonal X and Y vectors. Useful for transforming worldposition based on Hit normals etc.
 }
 nodes [
  n1 = FunctionInput(Preview=WorldPosition().XYZ) {bUsePreviewValueAsDefault=true, inputName=Vector to Transform, previewValue=<0 0 1 1>}
  n2 = FunctionInput() {
   bUsePreviewValueAsDefault=true
   description=Center Location will be subtracted from the Transformed Vector which is useful for centering the transformed vectors around something like ObjectPosition for example. Defaults to 0.
   inputName=Center Location
   sortPriority=2
  }
  n3 = FunctionInput() {bUsePreviewValueAsDefault=true, inputName=Z Vector, sortPriority=1, previewValue=<0 0.01 1 1>}
  n4 = Normalize(VectorInput=@n3) {desc=Z}
  n5 = Normalize(VectorInput=CrossProduct(A=@n4, B=Constant3Vector() {constant=<rgba 0 0 1 0>})) {desc=X}
  n6 = CrossProduct(A=@n4, B=@n5)
  FunctionOutput(None=MaterialFunctionCall!(VectorToTransform=Subtract(A=@n1, B=@n2), BasisX=@n5, BasisY=Normalize(VectorInput=@n6) {desc=Y}, BasisZ=@n4) {materialFunction=/Engine/Functions/Engine_MaterialFunctions02/Math/InverseTransformMatrix}.Transformed Vector 3)
 ]
}
