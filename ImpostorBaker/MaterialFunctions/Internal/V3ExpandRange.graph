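# Internal Impostor Baker material function that expands the range of a three-component vector's values.
# It divides the input by its maximum, then runs each channel through AlphaOffset using the original input's minimum as the offset, and recombines them.
# Both maximum and minimum compare only R and B, as R is wired twice where G would be expected.
# The offset comes from the unnormalized input.
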MaterialFunction /ImpostorBaker/MaterialFunctions/Internal/V3ExpandRange
graph {nodes [
  n1 = FunctionInput() {inputName=V3}
  n2 = MaterialFunctionCall(Float3=@n1) {materialFunction=&BreakOutFloat3Components}
  n3 = Max(A=@n2.R, B=@n2.R)
  n4 = MaterialFunctionCall(Float3=Divide(A=@n1, B=Max(A=@n3, B=@n2.B))) {materialFunction=&BreakOutFloat3Components}
  n5 = MaterialFunctionCall(Float3=@n1) {materialFunction=&BreakOutFloat3Components}
  n6 = Min(A=Min(A=@n5.R, B=@n5.R), B=@n5.B)
  FunctionOutput(None=MaterialFunctionCall(
    X=MaterialFunctionCall!(Alpha=@n4.R, Offset=@n6) {materialFunction=/Engine/Functions/Engine_MaterialFunctions01/Opacity/AlphaOffset}.Result
    Y=MaterialFunctionCall(Alpha=@n4.G, Offset=@n6) {materialFunction=&AlphaOffset}.Result
    Z=MaterialFunctionCall(Alpha=@n4.B, Offset=@n6) {materialFunction=&AlphaOffset}.Result
   ) {materialFunction=&MakeFloat3}.Result)
 ]}
