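# Material function for the Impostor Baker that optionally keeps an impostor's UVs inside the bounds of their atlas frame, so parallax does not push them outside it.
# Frame Index and the index plus one are each divided by FramesXY to form the min and max UV for a clamp.
# The "Clamp Frame UVs" static switch, off by default because distant views rarely need it, selects clamped or unchanged UVs.
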MaterialFunction /ImpostorBaker/MaterialFunctions/Internal/ClampImpostorFrameUVs
graph {nodes [
  n1 = FunctionInput() {inputName=UV, inputType=FunctionInput_Vector2, sortPriority=1}
  n2 = FunctionInput() {inputName=Frame Index, inputType=FunctionInput_Vector2, sortPriority=1}
  n3 = FunctionInput() {inputName=FramesXY, inputType=FunctionInput_Vector2, sortPriority=2}
  n4 = Add(A=@n2)
  n5 = StaticSwitchParameter(True=Clamp(None=@n1, Min=Divide(A=@n2, B=@n3) {desc=Min UV}, Max=Divide(A=@n4, B=@n3) {desc=Max UV}), False=@n1) {
   desc=Clamps the UVs for each frame from leaving the frame bounds due to parallax. This adds around 10 instructions which is around 3% of the cost. Not usually necessary when viewing from far way so default to False.
   parameterName=Clamp Frame UVs
  }
  FunctionOutput(None=@n5)
 ]}
