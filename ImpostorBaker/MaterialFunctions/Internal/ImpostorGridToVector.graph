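# Material function in ImpostorBaker that converts between a 2D grid value and a direction vector, in both directions.
# A static bool, Upper Hemisphere Only, switches each output between hemi-octahedral and full octahedral mapping.
# In hemisphere mode, the vector's Z is clamped to a small positive minimum and renormalized before encoding.
# That rebuilt vector takes both X and Y from the input's R component, leaving G unused.
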MaterialFunction /ImpostorBaker/MaterialFunctions/Internal/ImpostorGridToVector
graph {nodes [
  n1 = FunctionInput() {bUsePreviewValueAsDefault=true, inputName=2D Grid Value, inputType=FunctionInput_Vector2}
  n2 = FunctionInput(Preview=StaticBool()) {bUsePreviewValueAsDefault=true, inputName=Upper Hemisphere Only, inputType=FunctionInput_StaticBool, sortPriority=2}
  FunctionOutput(None=StaticSwitch(
    True=MaterialFunctionCall!(Hemi Octahedron=@n1) {materialFunction=&/HemiOctahedronToUnitVector}.Unit Vector
    False=MaterialFunctionCall!(Octahedron=@n1) {materialFunction=/Engine/Functions/Engine_MaterialFunctions01/Vectors/OctahedronToUnitVector}.Unit Vector
    Value=@n2
   )) {outputName=Vector}
  n3 = FunctionInput() {bUsePreviewValueAsDefault=true, inputName=Vector, sortPriority=1}
  n4 = MaterialFunctionCall(Float3=@n3) {materialFunction=&BreakOutFloat3Components}
  n5 = Normalize(VectorInput=MaterialFunctionCall(X=@n4.R, Y=@n4.R, Z=Max(A=@n4.B) {constB=0.001}) {materialFunction=&MakeFloat3}.Result)
  FunctionOutput(None=StaticSwitch(
    True=MaterialFunctionCall!(Unit Vector=@n5) {materialFunction=&/UnitVectorToHemiOctahedra}.Hemi Octahedron
    False=MaterialFunctionCall!(Vector=@n3) {materialFunction=/Engine/Functions/Engine_MaterialFunctions01/Vectors/UnitVectorToOctahedron}.Octahedron
    Value=@n2
   )) {outputName=2D Grid Value, sortPriority=1}
 ]}
