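# Material function in the ImpostorBaker plugin's internal helpers that splits a vector input into separate RG and B outputs.
# A single function input feeds two component masks, one keeping red and green as a two-channel value and the other keeping blue as a scalar.
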MaterialFunction /ImpostorBaker/MaterialFunctions/Internal/SeparateRG-B
graph {nodes [
  n1 = FunctionInput()
  FunctionOutput(None=ComponentMask(None=@n1) {g=true, r=true}) {outputName=RG}
  FunctionOutput(None=ComponentMask(None=@n1) {b=true}) {outputName=B, sortPriority=1}
 ]}
