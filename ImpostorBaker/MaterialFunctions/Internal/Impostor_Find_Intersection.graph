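# Impostor material function that applies depth-based parallax to atlas UVs, outputting UV, Alpha and Depth.
# Static switches choose iterative HLSL ray marching along RayDir with interpolated hit refinement, single-step bump offset, or no offset.
# Depth comes from a packed scalar map or normal-map alpha; the packed march reads red for depth regardless of the selected channel.
# A custom shadow-pass switch zeroes the parallax blend in shadow depth shaders.
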MaterialFunction /ImpostorBaker/MaterialFunctions/Internal/Impostor_Find_Intersection
graph {nodes [
  n1 = NamedRerouteUsage()
  MaterialExpressionNamedRerouteUsage_13 = NamedRerouteUsage()
  MaterialExpressionNamedRerouteUsage_14 = NamedRerouteUsage()
  MaterialExpressionNamedRerouteUsage_15 = NamedRerouteUsage()
  MaterialExpressionNamedRerouteUsage_17 = NamedRerouteUsage()
  MaterialExpressionNamedRerouteUsage_16 = NamedRerouteUsage()
  MaterialExpressionNamedRerouteUsage_18 = NamedRerouteUsage()
  MaterialExpressionNamedRerouteUsage_19 = NamedRerouteUsage()
  n2 = Custom!(
   Tex=@MaterialExpressionNamedRerouteUsage_13
   UV=@MaterialExpressionNamedRerouteUsage_14
   CamFwd=@MaterialExpressionNamedRerouteUsage_15
   Steps=@MaterialExpressionNamedRerouteUsage_17
   stepsize=@MaterialExpressionNamedRerouteUsage_16
   FramesXY=@MaterialExpressionNamedRerouteUsage_18
   Channel=@MaterialExpressionNamedRerouteUsage_19
  ) {
   code="float curdiff = 0;\r\nfloat lastdiff = 1;\r\nfloat timelerp = 0;\r\n\r\nfloat2 accum = 0;\r\nfloat4 cursample = 0;\r\nfloat rayZ = 0;\r\nfloat curZ = 0;\r\nfloat alpha = 0;\r\n\r\n\r\nfloat2 raystepUV = float2( (CamFwd.xy * stepsize)  / CamFwd.z / FramesXY);\r\nraystepUV *= float2(-1,-1);\r\n\r\n//float2 minuv = floor(UV * 12.0) / 12.0;\r\n//float2 maxuv = ceil(UV * 12.0) / 12.0;\r\n\r\nfloat starttest = dot(Tex.SampleLevel(TexSampler, UV, 0).rgba, Channel);\r\n\r\n//if( starttest > 0)\r\n{\r\nfor ( int i = 0; i < Steps;  i++)\r\n{\r\n\r\n\taccum += raystepUV;\r\n\trayZ += stepsize;\r\n\r\n\tfloat2 curUV = UV + accum.xy;\r\n\t//curUV = clamp( curUV, minuv, maxuv);\r\n\r\n\tcursample = Tex.SampleLevel(TexSampler, curUV, 0);\r\n\tcurZ = 1 - cursample.r;\r\n\talpha = cursample.a;\r\n\tcurdiff = rayZ - curZ;\r\n\r\n\tif ( curdiff > 0 &&&& alpha > 0)\r\n\t{\r\n\t\ttimelerp = saturate( lastdiff / (lastdiff - curdiff) );\r\n\t\taccum -= raystepUV * (1-timelerp);\r\n\t\treturn float4(accum, alpha, curZ) ;\r\n\t}\r\n\r\n\tlastdiff = curdiff;\r\n\r\n}\r\n}\r\n\r\n//float2 halfpoint = ((CamFwd.xy / CamFwd.z) * 0.5) / FramesXY;\r\nreturn float4(0,0,0,0);"
   desc=Use Packed Scalar Map Texture
   description=Intersection Trace
   outputType=CMOT_Float4
   inputs [
    {
     inputName=Tex
     input {expression=&self:MaterialExpressionNamedRerouteUsage_13}
    }
    {inputName=UV, input {expression=&self:MaterialExpressionNamedRerouteUsage_14}}
    {
     inputName=CamFwd
     input {expression=&self:MaterialExpressionNamedRerouteUsage_15}
    }
    {
     inputName=Steps
     input {expression=&self:MaterialExpressionNamedRerouteUsage_17}
    }
    {
     inputName=stepsize
     input {expression=&self:MaterialExpressionNamedRerouteUsage_16}
    }
    {
     inputName=FramesXY
     input {expression=&self:MaterialExpressionNamedRerouteUsage_18}
    }
    {
     inputName=Channel
     input {expression=&self:MaterialExpressionNamedRerouteUsage_19}
    }
   ]
  }
  MaterialExpressionNamedRerouteUsage_20 = NamedRerouteUsage()
  MaterialExpressionNamedRerouteUsage_21 = NamedRerouteUsage()
  MaterialExpressionNamedRerouteUsage_22 = NamedRerouteUsage()
  MaterialExpressionNamedRerouteUsage_24 = NamedRerouteUsage()
  MaterialExpressionNamedRerouteUsage_23 = NamedRerouteUsage()
  MaterialExpressionNamedRerouteUsage_25 = NamedRerouteUsage()
  n3 = Custom!(
   Tex=@MaterialExpressionNamedRerouteUsage_20
   UV=@MaterialExpressionNamedRerouteUsage_21
   CamFwd=@MaterialExpressionNamedRerouteUsage_22
   Steps=@MaterialExpressionNamedRerouteUsage_24
   stepsize=@MaterialExpressionNamedRerouteUsage_23
   FramesXY=@MaterialExpressionNamedRerouteUsage_25
  ) {
   code="float curdiff = 0;\r\nfloat lastdiff = 1;\r\nfloat timelerp = 0;\r\n\r\nfloat2 accum = 0;\r\nfloat4 cursample = 0;\r\nfloat rayZ = 0;\r\nfloat curZ = 0;\r\nfloat alpha = 0;\r\n\r\n\r\nfloat2 raystepUV = float2( (CamFwd.xy * stepsize)  / CamFwd.z / FramesXY);\r\nraystepUV *= float2(-1,-1);\r\n\r\n//float2 minuv = floor(UV * 12.0) / 12.0;\r\n//float2 maxuv = ceil(UV * 12.0) / 12.0;\r\n\r\nfloat starttest = Tex.SampleLevel(TexSampler, UV, 0).a;\r\n\r\n//if( starttest > 0)\r\n{\r\nfor ( int i = 0; i < Steps;  i++)\r\n{\r\n\r\n\taccum += raystepUV;\r\n\trayZ += stepsize;\r\n\r\n\tfloat2 curUV = UV + accum.xy;\r\n\t//curUV = clamp( curUV, minuv, maxuv);\r\n\r\n\tcursample = Tex.SampleLevel(TexSampler, curUV, 0).a;\r\n\tcurZ = 1 - cursample.r;\r\n\talpha = cursample.a;\r\n\tcurdiff = rayZ - curZ;\r\n\r\n\tif ( curdiff > 0)\r\n\t{\r\n\t\ttimelerp = saturate( lastdiff / (lastdiff - curdiff) );\r\n\t\taccum -= raystepUV * (1-timelerp);\r\n\t\treturn float4(accum, alpha, curZ) ;\r\n\t}\r\n\r\n\tlastdiff = curdiff;\r\n\r\n}\r\n}\r\n\r\n//float2 halfpoint = ((CamFwd.xy / CamFwd.z) * 0.5) / FramesXY;\r\nreturn float4(0,0,0,0);"
   desc=Use NormalMap Alpha
   description=Intersection Trace
   outputType=CMOT_Float4
   inputs [
    {
     inputName=Tex
     input {expression=&self:MaterialExpressionNamedRerouteUsage_20}
    }
    {inputName=UV, input {expression=&self:MaterialExpressionNamedRerouteUsage_21}}
    {
     inputName=CamFwd
     input {expression=&self:MaterialExpressionNamedRerouteUsage_22}
    }
    {
     inputName=Steps
     input {expression=&self:MaterialExpressionNamedRerouteUsage_24}
    }
    {
     inputName=stepsize
     input {expression=&self:MaterialExpressionNamedRerouteUsage_23}
    }
    {
     inputName=FramesXY
     input {expression=&self:MaterialExpressionNamedRerouteUsage_25}
    }
   ]
  }
  n4 = StaticSwitchParameter(True=@n2, False=@n3) {group=Channel Packed Masks, parameterName=UseDepthTexture}
  n5 = Add(A=@n1, B=NamedRerouteDeclaration(None=ComponentMask(None=@n4) {g=true, r=true}) {name=Iterative Parallax, nodeColor=<rgba 0 1 0.5078 1>})
  n6 = NamedRerouteUsage()
  n7 = TextureCoordinate() {uTiling=10, vTiling=10}
  n8 = TextureSample!(
   UVs=NamedRerouteUsage()
   Tex=NamedRerouteDeclaration(None=FunctionInput() {inputName=ScalarMaps, inputType=FunctionInput_Texture2D, sortPriority=1}) {name=ScalarMapsTex, nodeColor=<rgba 0 0 0 1>}
   "DDX(UVs)"=DDX(Value=@n7)
   "DDY(UVs)"=DDY(Value=@n7)
   Apply View MipBias
  ) {mipValueMode=TMVM_Derivative, texture=&Black}
  n9 = StaticSwitchParameter(True=@n8.B, False=StaticSwitchParameter(True=@n8.A, False=Constant()) {group=Channel Packed Masks, parameterName=Depth_Channel_A}) {group=Channel Packed Masks, parameterName=Depth_Channel_B}
  n10 = FunctionInput() {inputName=NormalTex, inputType=FunctionInput_Texture2D}
  n11 = StaticSwitchParameter(
   True=StaticSwitchParameter(True=@n8.R, False=StaticSwitchParameter(True=@n8.G, False=@n9) {group=Channel Packed Masks, parameterName=Depth_Channel_G}) {defaultValue=true, group=Channel Packed Masks, parameterName=Depth_Channel_R}
   False=TextureSample!(UVs=NamedRerouteUsage(), Tex=NamedRerouteDeclaration(None=@n10) {name=NormalTex, nodeColor=<rgba 0 0 0 1>}, Apply View MipBias) {texture=&Black}.A
  ) {group=Channel Packed Masks, parameterName=UseDepthTexture}
  n12 = NamedRerouteUsage()
  n13 = Divide(A=Multiply(A=Subtract(B=@n11) {constA=0.5, constB=0.5}, B=ComponentMask(None=@n12) {g=true, r=true}) {constB=0}, B=NamedRerouteUsage())
  n14 = NamedRerouteUsage()
  n15 = Add(A=Multiply(A=NamedRerouteDeclaration(None=@n13) {name=Simple Parallax, nodeColor=<rgba 1 0 0.1875 1>}, B=NamedRerouteUsage()), B=@n14)
  FunctionOutput(None=StaticSwitchParameter(
    True=LinearInterpolate(A=@n1, B=@n5, Alpha=@n6)
    False=StaticSwitchParameter(True=@n15, False=@n14) {defaultValue=true, parameterName=BumpOffset Parallax}
   ) {parameterName=Multi Iteration Parallax}) {outputName=UV}
  FunctionOutput(None=ComponentMask(None=@n4) {b=true}) {description=Alpha Amount. Only outputs when using Multi Iteration Parallax mode., outputName=Alpha, sortPriority=1}
  n16 = ComponentMask(None=@n4) {a=true}
  FunctionOutput(None=StaticSwitchParameter(True=OneMinus(None=@n16), False=@n11) {parameterName=Multi Iteration Parallax}) {outputName=Depth, sortPriority=2}
  n17 = FunctionInput() {inputName=UVs, inputType=FunctionInput_Vector2, sortPriority=2}
  n18 = MaterialFunctionCall!(In=NamedRerouteDeclaration(None=FunctionInput() {inputName=RayDir, sortPriority=3}) {name=RayDir, nodeColor=<rgba 0 0 0 1>}) {materialFunction=&/SeparateRG-B}
  n19 = Divide(A=@n18.RG, B=@n18.RG)
  n20 = StaticSwitchParameter(True=Multiply(A=Multiply(A=@n19) {constB=0.5}, B=NamedRerouteUsage()), False=Constant()) {parameterName=Multi Iteration Parallax}
  NamedRerouteDeclaration(None=Divide(A=Add(A=@n17, B=@n20), B=NamedRerouteUsage()) {constB=16}) {name=UVs, nodeColor=<rgba 0 0 0 1>}
  n21 = Min(A=ScalarParameter() {defaultValue=64, parameterName=ParallaxIterations}) {constB=256}
  NamedRerouteDeclaration(None=@n21) {name=NumSteps, nodeColor=<rgba 0 0 0 1>}
  MaterialExpressionStaticSwitchParameter_7 = StaticSwitchParameter(True=ScalarParameter() {defaultValue=1, parameterName=EnableParallax}, False=Constant() {r=1}) {desc=Useful to Toggle Parallax without recompiling, parameterName=Parallax Debug Toggle}
  MaterialExpressionConstant_5 = Constant()
  n22 = Custom!(Main=@MaterialExpressionStaticSwitchParameter_7, Shadow=@MaterialExpressionConstant_5) {
   code="##if SHADOW_DEPTH_SHADER \r\n{\r\nreturn Shadow;\r\n}\r\n##endif\r\nreturn Main;"
   desc=Optimize Shadows
   description=Shadow Pass Switch
   outputType=CMOT_Float2
   inputs [
    {
     inputName=Main
     input {expression=&self:MaterialExpressionStaticSwitchParameter_7}
    }
    {inputName=Shadow, input {expression=&self:MaterialExpressionConstant_5}}
   ]
  }
  NamedRerouteDeclaration(None=@n22) {name=Enable Parallax, nodeColor=<rgba 0 0.4375 1 1>}
  NamedRerouteDeclaration(None=FunctionInput() {inputName=FramesXY, inputType=FunctionInput_Vector2, sortPriority=4}) {name=FramesXY, nodeColor=<rgba 0 0 0 1>}
  NamedRerouteDeclaration(None=Divide(B=@n21)) {name=StepSize, nodeColor=<rgba 0 0 0 1>}
  n23 = Constant4Vector() {constant=<rgba 0 1 0 0>}
  n24 = StaticSwitchParameter(
   True=Constant4Vector() {constant=<rgba 0 0 1 0>}.RGBA
   False=StaticSwitchParameter(True=Constant4Vector() {constant=<rgba 0 1 0 0>}.RGBA, False=Constant()) {group=Channel Packed Masks, parameterName=Depth_Channel_A}
  ) {group=Channel Packed Masks, parameterName=Depth_Channel_B}
  NamedRerouteDeclaration(None=StaticSwitchParameter(
    True=Constant4Vector() {constant=<rgba 1 0 0 0>}.RGBA
    False=StaticSwitchParameter(True=@n23.RGBA, False=@n24) {group=Channel Packed Masks, parameterName=Depth_Channel_G}
   ) {defaultValue=true, group=Channel Packed Masks, parameterName=Depth_Channel_R}) {name=DepthChannel, nodeColor=<rgba 0 0 0 1>}
  NamedRerouteUsage()
 ]}
